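# Modular Control Rig module that builds an FK control, with offset null and visibility channel, for each bone resolved by its Bones connector.
# Construction skips missing bones, applies override parents, and creates space and orient-space nulls from connectors.
# Forward solve updates visibility and spaces, then optionally drives bones; Use Active Skeleton layers local control transforms over the current pose without propagating to children.
# Backward solve snaps controls to bones.
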ControlRigBlueprint /EpicControlRigPub/v5_8/Modules/FkSolves/Fk/CRM_FkArray_v02
graphs [
 ConstructionGraph {nodes [
   RigUnit_PrepareForExecution = &RigUnit_PrepareForExecution()
   n1 = /Script/ControlRig.RigUnit_SetupShapeLibraryFromUserData(LibraryName=CRSL, LogShapeLibraries=False, NameSpace=CRSL, Path=ShapeLibrary, ExecutePin=@RigUnit_PrepareForExecution.ExecutePin) {is=struct RigVMUnitNode}
   > RigVMVariableNode(Value=@n1.bFound) {variable=CRSL Found}
   > RigVMDispatchNode(Array=RigVMVariableNode() {variable=Bones}.Value) {template="DISPATCH_RigVMDispatch_ArrayReset(io Array)"}
   > DISPATCH_RigVMDispatch_ArrayReset = RigVMDispatchNode(Array=RigVMVariableNode() {variable=Controls}.Value) {template="DISPATCH_RigVMDispatch_ArrayReset(io Array)"}
   > RigVMDispatchNode(Array=RigVMVariableNode() {variable=Fk Parents}.Value) {template="DISPATCH_RigVMDispatch_ArrayReset(io Array)"}
   > RigVMDispatchNode(Array=RigVMVariableNode() {variable=Vis Channels}.Value) {template="DISPATCH_RigVMDispatch_ArrayReset(io Array)"}
   > RigVMDispatchNode(Array=RigVMVariableNode() {variable=Spaces}.Value) {template="DISPATCH_RigVMDispatch_ArrayReset(io Array)"}
   > RigVMDispatchNode(Array=RigVMVariableNode() {variable=Orient Spaces}.Value) {template="DISPATCH_RigVMDispatch_ArrayReset(io Array)"}
   > RigVMDispatchNode(Array=RigVMVariableNode() {variable=Controls Visibility Initials Sorted}.Value) {template="DISPATCH_RigVMDispatch_ArrayReset(io Array)"}
   > RigVMDispatchNode(Array=RigVMVariableNode() {variable=Display Names Sorted}.Value) {template="DISPATCH_RigVMDispatch_ArrayReset(io Array)"}
   > RigVMFunction_Sequence = RigVMAggregateNode()
   RigVMDispatchNode(Array=RigVMVariableNode() {variable=Orient Space Nulls}.Value, ExecuteContext=@RigVMFunction_Sequence.F) {template="DISPATCH_RigVMDispatch_ArrayReset(io Array)"}
   > RigVMDispatchNode(Array=RigVMVariableNode() {variable=Combined Orient Spaces}.Value) {template="DISPATCH_RigVMDispatch_ArrayReset(io Array)"}
   > For_Each_1 = RigVMDispatchNode(Array=RigVMVariableNode() {variable=Bones}.Value) {template="DISPATCH_RigVMDispatch_ArrayIterator(in Array,out Element,out Index,out Count,out Ratio)"}
   n2 = RigVMVariableNode() {variable=Orient Spaces}
   > Construct Space Nulls v01_3 = RigVMFunctionReferenceNode(
    Driven Control=RigVMDispatchNode(Index=@For_Each_1.Index, Array=RigVMVariableNode() {variable=Controls}.Value) {template="DISPATCH_RigVMDispatch_ArrayGetAtIndex(in Array,in Index,out Element)"}.Element
    Space Null Parent=/Script/ControlRig.RigUnit_HierarchyGetParent(bDefaultParent=True, Child=@For_Each_1.Element) {is=struct RigVMUnitNode}.Parent
    Spaces=@n2.Value
   )
   > RigVMDispatchNode(Other=@"Construct Space Nulls v01_3".Output Space Nulls, Array=RigVMVariableNode() {variable=Orient Space Nulls}.Value) {template="DISPATCH_RigVMDispatch_ArrayAppend(io Array,in Other)"}
   > RigVMDispatchNode(Array=RigVMVariableNode() {variable=Combined Orient Spaces}.Value, Other=@n2.Value) {template="DISPATCH_RigVMDispatch_ArrayAppend(io Array,in Other)"}
   n3 = &RigVMFunction_ControlFlowBranch(Condition=RigVMVariableNode() {variable=CRSL Found}.Value, ExecuteContext=@RigVMFunction_Sequence.G)
   Get_Module_Metadata_1 = RigVMDispatchNode(Default=1, Name=Global Control Scale, NameSpace=Root) {template="DISPATCH_RigDispatch_GetModuleMetadata(in Name,in NameSpace,in Default,out Value,out Found)"}
   Set Control Scale_1 = RigVMFunctionReferenceNode(
    Auto Scale Factor=1
    Global Scale=@Get_Module_Metadata_1.Value
    ExecuteContext=@n3.False
    Controls=RigVMVariableNode() {variable=Controls}.Value
    Scale=RigVMVariableNode() {variable=Control Scale}.Value
   )
   RigVMDispatchNode(Array=RigVMVariableNode() {variable=Space Nulls}.Value, ExecuteContext=@RigVMFunction_Sequence.E) {template="DISPATCH_RigVMDispatch_ArrayReset(io Array)"}
   > RigVMDispatchNode(Array=RigVMVariableNode() {variable=Combined Spaces}.Value) {template="DISPATCH_RigVMDispatch_ArrayReset(io Array)"}
   > For_Each = RigVMDispatchNode(Array=RigVMVariableNode() {variable=Bones}.Value) {template="DISPATCH_RigVMDispatch_ArrayIterator(in Array,out Element,out Index,out Count,out Ratio)"}
   n4 = RigVMVariableNode() {variable=Spaces}
   > Construct Space Nulls v01_2 = RigVMFunctionReferenceNode(
    Driven Control=RigVMDispatchNode(Index=@For_Each.Index, Array=RigVMVariableNode() {variable=Controls}.Value) {template="DISPATCH_RigVMDispatch_ArrayGetAtIndex(in Array,in Index,out Element)"}.Element
    Space Null Parent=&RigUnit_HierarchyGetParent(bDefaultParent=True, Child=@For_Each.Element).Parent
    Spaces=@n4.Value
   )
   > RigVMDispatchNode(Other=@"Construct Space Nulls v01_2".Output Space Nulls, Array=RigVMVariableNode() {variable=Space Nulls}.Value) {template="DISPATCH_RigVMDispatch_ArrayAppend(io Array,in Other)"}
   > RigVMDispatchNode(Array=RigVMVariableNode() {variable=Combined Spaces}.Value, Other=@n4.Value) {template="DISPATCH_RigVMDispatch_ArrayAppend(io Array,in Other)"}
   DISPATCH_RigVMDispatch_ArrayIterator = RigVMDispatchNode(Array=RigVMVariableNode() {variable=Bones}.Value, ExecuteContext=@RigVMFunction_Sequence.D) {template="DISPATCH_RigVMDispatch_ArrayIterator(in Array,out Element,out Index,out Count,out Ratio)"}
   > Get Control Name = RigVMFunctionReferenceNode(
    Count=@DISPATCH_RigVMDispatch_ArrayIterator.Count
    Name=@DISPATCH_RigVMDispatch_ArrayIterator.Element.Name
    Index=@DISPATCH_RigVMDispatch_ArrayIterator.Index
    FK Standard Names=RigVMVariableNode() {variable=Standard FK Names}.Value
   )
   > n5 = &RigUnit_HierarchyAddNull(
    Space=GlobalSpace
    Parent=RigVMDispatchNode(Array=RigVMVariableNode() {variable=Fk Parents}.Value, Index=@DISPATCH_RigVMDispatch_ArrayIterator.Index) {template="DISPATCH_RigVMDispatch_ArrayGetAtIndex(in Array,in Index,out Element)"}.Element
    Name=@"Get Control Name".Result
    Transform=/Script/RigVM.RigVMFunction_MathTransformMakeAbsolute(
     Parent=&RigUnit_GetTransform(bInitial=true, Item=@DISPATCH_RigVMDispatch_ArrayIterator.Element).Transform
     Local=RigVMVariableNode() {variable=Control Transform Offset}.Value
    ) {is=struct RigVMUnitNode, template="Make Absolute::Execute(in Local,in Parent,out Global)"}.Global
   )
   > RigVMDispatchNode(Array=RigVMVariableNode() {variable=Nulls}.Value, Element=@n5.Item) {template="DISPATCH_RigVMDispatch_ArrayAdd(io Array,in Element,out Index)"}
   Get Control Shape = RigVMFunctionReferenceNode(
    Shape=RigVMVariableNode() {variable=FK Control Shape}.Value
    Transform Offset=RigVMVariableNode() {variable=Control Transform Offset}.Value
    Bone=@DISPATCH_RigVMDispatch_ArrayIterator.Element
   )
   n6 = RigVMVariableNode() {variable=Display Names Sorted}
   n7 = RigVMDispatchNode(Array=@n6.Value) {template="DISPATCH_RigVMDispatch_ArrayGetNum(in Array,out Num)"}
   GreaterEqual = /Script/RigVM.RigVMFunction_MathIntGreaterEqual(B=1, A=@n7.Num) {is=struct RigVMUnitNode, template="GreaterEqual::Execute(in A,in B,out Result)"}
   Less = /Script/RigVM.RigVMFunction_MathIntLess(B=@n7.Num, A=@DISPATCH_RigVMDispatch_ArrayIterator.Index) {is=struct RigVMUnitNode, template="Less::Execute(in A,in B,out Result)"}
   RigVMFunction_MathBoolAnd = &RigVMFunction_MathBoolAnd(B=@GreaterEqual.Result, A=@Less.Result)
   RigVMFunction_AnimEvalRichCurve = &RigVMFunction_AnimEvalRichCurve(
    SourceMinimum=0
    TargetMinimum=0
    Curve=RigVMVariableNode() {variable=Control Scale Factor Profile}.Value
    Value=@DISPATCH_RigVMDispatch_ArrayIterator.Ratio
   )
   Scale = /Script/RigVM.RigVMFunction_MathVectorScale(Value=@"Get Control Shape".Result.Transform.Scale3D, Factor=@RigVMFunction_AnimEvalRichCurve.Result) {is=struct RigVMUnitNode, template="Scale::Execute(in Value,in Factor,out Result)"}
   > SpawnControl = /Script/ControlRig.RigUnit_HierarchyAddControlTransform(
    InitialValue=(Rotation=(X=0,Y=0,Z=0,W=1),Translation=(X=0,Y=0,Z=0),Scale3D=(X=1,Y=1,Z=1))
    OffsetSpace=LocalSpace
    OffsetTransform=(Rotation=(X=0,Y=0,Z=0,W=1),Translation=(X=0,Y=0,Z=0),Scale3D=(X=1,Y=1,Z=1))
    Settings="(InitialSpace=LocalSpace,bUsePreferredRotationOrder=true,PreferredRotationOrder=ZYX,Limits=(LimitTranslationX=(bMinimum=False,bMaximum=False),LimitTranslationY=(bMinimum=False,bMaximum=False),LimitTranslationZ=(bMinimum=False,bMaximum=False),LimitPitch=(bM ...(truncated)"
    Name=@"Get Control Name".Result
    Settings.Shape.Color=@"Get Control Shape".Result.Color
    Settings.Shape.Name=@"Get Control Shape".Result.Name
    Settings.Shape.Transform.Rotation=@"Get Control Shape".Result.Transform.Rotation
    Settings.Shape.Transform.Translation=@"Get Control Shape".Result.Transform.Translation
    Settings.Shape.bVisible=@"Get Control Shape".Result.bVisible
    Settings.DisplayName=RigVMDispatchNode(
     False=None
     True=RigVMDispatchNode(Array=@n6.Value, Index=@DISPATCH_RigVMDispatch_ArrayIterator.Index) {template="DISPATCH_RigVMDispatch_ArrayGetAtIndex(in Array,in Index,out Element)"}.Element
     Condition=@RigVMFunction_MathBoolAnd.Result
    ) {template="DISPATCH_RigVMDispatch_If(in Condition,in True,in False,out Result)"}.Result
    Settings.Shape.Transform.Scale3D=@Scale.Result
    Settings.PreferredRotationOrder=RigVMVariableNode() {variable=Rotation Order}.Value
    Parent=@n5.Item
   ) {
    is=struct RigVMUnitNode
    template="SpawnControl::Execute(in InitialValue,in Settings,in OffsetTransform,in OffsetSpace,in Parent,in Name,out Item)"
   }
   > Set_At_1 = RigVMDispatchNode(Element=@SpawnControl.Item, Array=RigVMVariableNode() {variable=Fk Parents}.Value, Index=@DISPATCH_RigVMDispatch_ArrayIterator.Index) {template="DISPATCH_RigVMDispatch_ArraySetAtIndex(io Array,in Index,in Element)"}
   > Set Mirror Axis_1 = RigVMFunctionReferenceNode(Value=RigVMVariableNode() {variable=Mirror Axis}.Value, Item=@SpawnControl.Item)
   > Set Mirror Behavior_1 = RigVMFunctionReferenceNode(Value=RigVMVariableNode() {variable=Mirror Behavior}.Value, Item=@SpawnControl.Item)
   > DISPATCH_RigVMDispatch_ArrayAdd = RigVMDispatchNode(Array=RigVMVariableNode() {variable=Controls}.Value, Element=@SpawnControl.Item) {template="DISPATCH_RigVMDispatch_ArrayAdd(io Array,in Element,out Index)"}
   n8 = RigVMVariableNode() {variable=Controls Visibility Initials Sorted}
   Num = RigVMDispatchNode(Array=@n8.Value) {template="DISPATCH_RigVMDispatch_ArrayGetNum(in Array,out Num)"}
   n9 = &RigVMFunction_MathIntGreaterEqual(A=@Num.Num, B=@DISPATCH_RigVMDispatch_ArrayIterator.Index) {template="GreaterEqual::Execute(in A,in B,out Result)"}
   Not_Equals = RigVMDispatchNode(B=0, A=@Num.Num) {template="DISPATCH_RigVMDispatch_CoreNotEquals(in A,in B,out Result)"}
   > SpawnAnimationChannel = /Script/ControlRig.RigUnit_HierarchyAddAnimationChannelBool(
    LimitsEnabled=()
    MaximumValue=True
    MinimumValue=False
    Name=Visibility
    InitialValue=RigVMDispatchNode(
     False=true
     True=RigVMDispatchNode(Array=@n8.Value, Index=@DISPATCH_RigVMDispatch_ArrayIterator.Index) {template="DISPATCH_RigVMDispatch_ArrayGetAtIndex(in Array,in Index,out Element)"}.Element
     Condition=&RigVMFunction_MathBoolAnd(B=@n9.Result, A=@Not_Equals.Result).Result
    ) {template="DISPATCH_RigVMDispatch_If(in Condition,in True,in False,out Result)"}.Result
    Parent=@SpawnControl.Item
   ) {
    is=struct RigVMUnitNode
    template="SpawnAnimationChannel::Execute(in InitialValue,in MinimumValue,in MaximumValue,in LimitsEnabled,in Parent,in Name,out Item)"
   }
   > RigVMDispatchNode(Element=@SpawnAnimationChannel.Item, Array=RigVMVariableNode() {variable=Vis Channels}.Value) {template="DISPATCH_RigVMDispatch_ArrayAdd(io Array,in Element,out Index)"}
   For_Each_3 = RigVMDispatchNode(Array=RigVMVariableNode() {variable=Bones}.Value, ExecuteContext=@RigVMFunction_Sequence.C) {template="DISPATCH_RigVMDispatch_ArrayIterator(in Array,out Element,out Index,out Count,out Ratio)"}
   > Conform Name v01_1 = RigVMFunctionReferenceNode(Name=@For_Each_3.Element.Name)
   Concat = /Script/RigVM.RigVMFunction_NameConcat(B=" Default Space", A=@"Conform Name v01_1".Conformed Name) {is=struct RigVMUnitNode, template="Concat::Execute(in A,in B,out Result)"}
   > n10 = &RigUnit_HierarchyAddNull(
    Space=GlobalSpace
    Parent=RigVMDispatchNode(Array=RigVMVariableNode() {variable=Fk Parents}.Value, Index=@For_Each_3.Index) {template="DISPATCH_RigVMDispatch_ArrayGetAtIndex(in Array,in Index,out Element)"}.Element
    Name=@Concat.Result
    Transform=&RigUnit_GetTransform(bInitial=true, Item=@For_Each_3.Element).Transform
   )
   > Set_At = RigVMDispatchNode(Element=@n10.Item, Array=RigVMVariableNode() {variable=Fk Parents}.Value, Index=@For_Each_3.Index) {template="DISPATCH_RigVMDispatch_ArraySetAtIndex(io Array,in Index,in Element)"}
   For_Each_2 = RigVMDispatchNode(
    Array=/Script/ControlRig.RigUnit_ResolveArrayConnector(Connector=(Type=Connector,Name="Bones"), SkipSocket=False) {is=struct RigVMUnitNode}.Result
    ExecuteContext=@RigVMFunction_Sequence.A
   ) {template="DISPATCH_RigVMDispatch_ArrayIterator(in Array,out Element,out Index,out Count,out Ratio)"}
   RigVMFunction_ControlFlowBranch = &RigVMFunction_ControlFlowBranch(
    ExecuteContext=@For_Each_2.ExecuteContext
    Condition=/Script/ControlRig.RigUnit_ItemExists(Item=@For_Each_2.Element) {is=struct RigVMUnitNode}.Exists
   )
   RigVMDispatchNode(
    Element=@For_Each_2.Index
    ExecuteContext=@RigVMFunction_ControlFlowBranch.False
    Array=RigVMVariableNode() {variable=Indices to Remove}.Value
   ) {template="DISPATCH_RigVMDispatch_ArrayAdd(io Array,in Element,out Index)"}
   RigVMDispatchNode(ExecuteContext=@RigVMFunction_ControlFlowBranch.True, Array=RigVMVariableNode() {variable=Bones}.Value, Element=@For_Each_2.Element) {template="DISPATCH_RigVMDispatch_ArrayAdd(io Array,in Element,out Index)"}
   Get Array Parents_1 = RigVMFunctionReferenceNode(ExecuteContext=@For_Each_2.Completed, Items=RigVMVariableNode() {variable=Bones}.Value)
   > RigVMVariableNode(Value=@"Get Array Parents_1".Parents) {variable=Fk Parents}
   > RigVMFunction_Sequence_1 = RigVMAggregateNode()
   For_Each_5 = RigVMDispatchNode(
    Array=&RigUnit_ResolveArrayConnector(Connector=(Type=Connector,Name="Override Parents"), SkipSocket=False).Result
    ExecuteContext=@RigVMFunction_Sequence_1.A
   ) {template="DISPATCH_RigVMDispatch_ArrayIterator(in Array,out Element,out Index,out Count,out Ratio)"}
   n11 = RigVMDispatchNode(Element=@For_Each_5.Index, Array=RigVMVariableNode() {variable=Indices to Remove}.Value) {template="DISPATCH_RigVMDispatch_ArrayFind(in Array,in Element,out Index,out Success)"}
   n12 = &RigVMFunction_ControlFlowBranch(
    ExecuteContext=@For_Each_5.ExecuteContext
    Condition=&RigVMFunction_MathBoolAnd(
     A=&RigUnit_ItemExists(Item=@For_Each_5.Element).Exists
     B=&RigVMFunction_MathBoolNot(Value=@n11.Success).Result
    ).Result
   )
   Set_At_2 = RigVMDispatchNode(ExecuteContext=@n12.True, Array=RigVMVariableNode() {variable=Fk Parents}.Value, Index=@For_Each_5.Index, Element=@For_Each_5.Element) {template="DISPATCH_RigVMDispatch_ArraySetAtIndex(io Array,in Index,in Element)"}
   For_Each_4 = RigVMDispatchNode(
    Array=&RigUnit_ResolveArrayConnector(Connector=(Type=Connector,Name="Spaces"), SkipSocket=False).Result
    ExecuteContext=@RigVMFunction_Sequence_1.B
   ) {template="DISPATCH_RigVMDispatch_ArrayIterator(in Array,out Element,out Index,out Count,out Ratio)"}
   n13 = &RigVMFunction_ControlFlowBranch(ExecuteContext=@For_Each_4.ExecuteContext, Condition=&RigUnit_ItemExists(Item=@For_Each_4.Element).Exists)
   RigVMDispatchNode(ExecuteContext=@n13.True, Array=RigVMVariableNode() {variable=Spaces}.Value, Element=@For_Each_4.Element) {template="DISPATCH_RigVMDispatch_ArrayAdd(io Array,in Element,out Index)"}
   For_Each_6 = RigVMDispatchNode(
    Array=&RigUnit_ResolveArrayConnector(Connector=(Type=Connector,Name="Orient Spaces"), SkipSocket=False).Result
    ExecuteContext=@RigVMFunction_Sequence_1.C
   ) {template="DISPATCH_RigVMDispatch_ArrayIterator(in Array,out Element,out Index,out Count,out Ratio)"}
   n14 = &RigVMFunction_ControlFlowBranch(ExecuteContext=@For_Each_6.ExecuteContext, Condition=&RigUnit_ItemExists(Item=@For_Each_6.Element).Exists)
   RigVMDispatchNode(Element=@For_Each_6.Element, ExecuteContext=@n14.True, Array=RigVMVariableNode() {variable=Orient Spaces}.Value) {template="DISPATCH_RigVMDispatch_ArrayAdd(io Array,in Element,out Index)"}
   For_Each_7 = RigVMDispatchNode(Array=RigVMVariableNode() {variable=Display Names}.Value, ExecuteContext=@RigVMFunction_Sequence_1.D) {template="DISPATCH_RigVMDispatch_ArrayIterator(in Array,out Element,out Index,out Count,out Ratio)"}
   n15 = RigVMVariableNode() {variable=Indices to Remove}
   n16 = &RigVMFunction_ControlFlowBranch(
    ExecuteContext=@For_Each_7.ExecuteContext
    Condition=&RigVMFunction_MathBoolNot(Value=RigVMDispatchNode(Element=@For_Each_7.Index, Array=@n15.Value) {template="DISPATCH_RigVMDispatch_ArrayFind(in Array,in Element,out Index,out Success)"}.Success).Result
   )
   RigVMDispatchNode(Element=@For_Each_7.Element, ExecuteContext=@n16.True, Array=RigVMVariableNode() {variable=Display Names Sorted}.Value) {template="DISPATCH_RigVMDispatch_ArrayAdd(io Array,in Element,out Index)"}
   For_Each_8 = RigVMDispatchNode(Array=RigVMVariableNode() {variable=Controls Visibility Initials}.Value, ExecuteContext=@RigVMFunction_Sequence_1.E) {template="DISPATCH_RigVMDispatch_ArrayIterator(in Array,out Element,out Index,out Count,out Ratio)"}
   n17 = RigVMVariableNode() {variable=Indices to Remove}
   n18 = &RigVMFunction_ControlFlowBranch(
    ExecuteContext=@For_Each_8.ExecuteContext
    Condition=&RigVMFunction_MathBoolNot(Value=RigVMDispatchNode(Element=@For_Each_8.Index, Array=@n17.Value) {template="DISPATCH_RigVMDispatch_ArrayFind(in Array,in Element,out Index,out Success)"}.Success).Result
   )
   RigVMDispatchNode(Element=@For_Each_8.Element, ExecuteContext=@n18.True, Array=RigVMVariableNode() {variable=Controls Visibility Initials Sorted}.Value) {template="DISPATCH_RigVMDispatch_ArrayAdd(io Array,in Element,out Index)"}
  ]}
 New Function_ContainedGraph {nodes [
   n1 = RigVMFunctionEntryNode()
   n2 = RigVMVariableNode() {variable=Shape}
   Control Color Override From Metadata v02_1 = RigVMFunctionReferenceNode(Module Color=(R=0,G=0,B=0,A=0), Bone=@n1.Bone, Color Override=@n2.Value.Color)
   Get Control Shape Name From Item v01 = RigVMFunctionReferenceNode(Item=(Type=None,Name="None"), ShapeLib Namespace=CRSL, Item.Name=@n1.Bone.Name, Default Shape=@n2.Value.Name)
   Multiply = &RigVMFunction_MathTransformMul(
    B=/Script/RigVM.RigVMFunction_MathTransformInverse(Value=@n1.Transform Offset) {is=struct RigVMUnitNode, template="Inverse::Execute(in Value,out Result)"}.Result
    A=@n2.Value.Transform
   ) {template="Multiply::Execute(in A,in B,out Result)"}
   > RigVMFunctionReturnNode(
    Result=(bVisible=True,Name="Default",Color=(R=1,G=0,B=0,A=1),Transform=(Rotation=(X=0,Y=0,Z=0,W=1),Translation=(X=0,Y=0,Z=0),Scale3D=(X=1,Y=1,Z=1)))
    Result.Color=@"Control Color Override From Metadata v02_1".Result
    Result.Name=@"Get Control Shape Name From Item v01".Result
    Result.Transform=@Multiply.Result
    Result.bVisible=@n2.Value.bVisible
   )
  ]}
 New Function_ContainedGraph {nodes [
   n1 = RigVMFunctionEntryNode()
   > Conform Name v01 = RigVMFunctionReferenceNode(Name=@n1.Name)
   Int to Name = /Script/RigVM.RigVMFunction_MathIntToName(PaddedSize=0, Number=@n1.Index) {is=struct RigVMUnitNode, template="Int to Name::Execute(in Number,in PaddedSize,out Result)"}
   Concat = &RigVMFunction_NameConcat(A="FK ", B=@"Int to Name".Result) {template="Concat::Execute(in A,in B,out Result)"}
   Greater = /Script/RigVM.RigVMFunction_MathIntGreater(B=1, A=@n1.Count) {is=struct RigVMUnitNode, template="Greater::Execute(in A,in B,out Result)"}
   DISPATCH_RigVMDispatch_If = RigVMDispatchNode(False=FK, True=@Concat.Result, Condition=@Greater.Result) {template="DISPATCH_RigVMDispatch_If(in Condition,in True,in False,out Result)"}
   > RigVMFunctionReturnNode(Result=RigVMDispatchNode(False=@"Conform Name v01".Conformed Name, True=@DISPATCH_RigVMDispatch_If.Result, Condition=@n1.FK Standard Names) {template="DISPATCH_RigVMDispatch_If(in Condition,in True,in False,out Result)"}.Result)
  ]}
 RigVMFunctionLibrary {nodes [
   Get Control Name = RigVMCollapseNode(Count=0, ExecuteContext=(), FK Standard Names=False, Index=0)
   Get Control Shape = RigVMCollapseNode(
    Bone=(Type=None,Name="None")
    Shape=(bVisible=True,Name="Default",Color=(R=1,G=0,B=0,A=1),Transform=(Rotation=(X=0,Y=0,Z=0,W=1),Translation=(X=0,Y=0,Z=0),Scale3D=(X=1,Y=1,Z=1)))
    Transform Offset=(Rotation=(X=0,Y=0,Z=0,W=1),Translation=(X=0,Y=0,Z=0),Scale3D=(X=1,Y=1,Z=1))
   )
  ]}
 RigVMFunction_Sequence_1_ContainedGraph {nodes [
   n1 = RigVMFunctionEntryNode()
   RigVMFunction_Sequence = &RigVMFunction_Sequence(ExecuteContext=@n1.ExecuteContext)
   n2 = &RigVMFunction_Sequence(ExecuteContext=@RigVMFunction_Sequence.B)
   n3 = &RigVMFunction_Sequence(ExecuteContext=@n2.B)
   n4 = &RigVMFunction_Sequence(ExecuteContext=@n3.B)
   n5 = &RigVMFunction_Sequence(ExecuteContext=@n4.B)
   RigVMFunctionReturnNode(A=@RigVMFunction_Sequence.A, C=@n2.A, D=@n3.A, E=@n4.A, F=@n5.A, G=@n5.B)
  ]}
 RigVMFunction_Sequence_2_ContainedGraph {nodes [
   n1 = RigVMFunctionEntryNode()
   RigVMFunction_Sequence = &RigVMFunction_Sequence(ExecuteContext=@n1.ExecuteContext)
   n2 = &RigVMFunction_Sequence(ExecuteContext=@RigVMFunction_Sequence.B)
   RigVMFunctionReturnNode(A=@RigVMFunction_Sequence.A, B=@n2.A, C=@n2.B)
  ]}
 RigVMFunction_Sequence_2_ContainedGraph {nodes [
   n1 = RigVMFunctionEntryNode()
   n2 = &RigVMFunction_Sequence(ExecuteContext=@n1.ExecuteContext)
   n3 = &RigVMFunction_Sequence(ExecuteContext=@n2.B)
   n4 = &RigVMFunction_Sequence(ExecuteContext=@n3.B)
   n5 = &RigVMFunction_Sequence(ExecuteContext=@n4.B)
   RigVMFunctionReturnNode(A=@n2.A, B=@n3.A, C=@n4.A, D=@n5.A, E=@n5.B)
  ]}
 RigVMModel {nodes [
   RigUnit_BeginExecution = &RigUnit_BeginExecution()
   RigVMFunction_Sequence = RigVMAggregateNode(ExecuteContext=@RigUnit_BeginExecution.ExecutePin)
   For_Each = RigVMDispatchNode(ExecuteContext=@RigVMFunction_Sequence.A, Array=RigVMVariableNode() {variable=Controls}.Value) {template="DISPATCH_RigVMDispatch_ArrayIterator(in Array,out Element,out Index,out Count,out Ratio)"}
   GetAnimationChannelFromItem = /Script/ControlRig.RigUnit_GetBoolAnimationChannelFromItem(
    bInitial=False
    Item=RigVMDispatchNode(Index=@For_Each.Index, Array=RigVMVariableNode() {variable=Vis Channels}.Value) {template="DISPATCH_RigVMDispatch_ArrayGetAtIndex(in Array,in Index,out Element)"}.Element
   ) {is=struct RigVMUnitNode, template="GetAnimationChannelFromItem::Execute(out Value,in Item,in bInitial)"}
   > SetControlVisibility = /Script/ControlRig.RigUnit_SetControlVisibility(Item=@For_Each.Element, bVisible=@GetAnimationChannelFromItem.Value) {is=struct RigVMUnitNode, template="SetControlVisibility::Execute(in Item,in Pattern,in bVisible)"}
   n1 = &RigVMFunction_Sequence(ExecuteContext=@RigVMFunction_Sequence.B)
   Evaluate Space Nulls v01_2 = RigVMFunctionReferenceNode(
    Orient Only=false
    ExecuteContext=@n1.A
    Space Nulls=RigVMVariableNode() {variable=Orient Space Nulls}.Value
    Spaces=RigVMVariableNode() {variable=Combined Orient Spaces}.Value
   )
   Evaluate Space Nulls v01_3 = RigVMFunctionReferenceNode(
    Orient Only=false
    ExecuteContext=@n1.B
    Space Nulls=RigVMVariableNode() {variable=Space Nulls}.Value
    Spaces=RigVMVariableNode() {variable=Combined Spaces}.Value
   )
   RigVMFunction_ControlFlowBranch = &RigVMFunction_ControlFlowBranch(Condition=RigVMVariableNode() {variable=Drive Bones}.Value, ExecuteContext=@RigVMFunction_Sequence.C)
   DISPATCH_RigVMDispatch_ArrayIterator = RigVMDispatchNode(ExecuteContext=@RigVMFunction_ControlFlowBranch.True, Array=RigVMVariableNode() {variable=Bones}.Value) {template="DISPATCH_RigVMDispatch_ArrayIterator(in Array,out Element,out Index,out Count,out Ratio)"}
   n2 = &RigVMFunction_ControlFlowBranch(ExecuteContext=@DISPATCH_RigVMDispatch_ArrayIterator.ExecuteContext, Condition=RigVMVariableNode() {variable=Use Active Skeleton}.Value)
   n3 = RigVMVariableNode() {variable=Control Transform Offset}
   RigVMFunction_MathBoolNot = &RigVMFunction_MathBoolNot(Value=RigVMVariableNode() {variable=Use Active Skeleton}.Value)
   &RigUnit_SetTransform(
    Weight=1
    bInitial=False
    Item=@DISPATCH_RigVMDispatch_ArrayIterator.Element
    Value=&RigVMFunction_MathTransformMakeAbsolute(
     Local=&RigVMFunction_MathTransformInverse(Value=@n3.Value) {template="Inverse::Execute(in Value,out Result)"}.Result
     Parent=RigVMVariableNode() {variable=Bone Transform}.Value
    ) {template="Make Absolute::Execute(in Local,in Parent,out Global)"}.Global
    ExecutePin=@n2.Completed
    bPropagateToChildren=@RigVMFunction_MathBoolNot.Result
   )
   DISPATCH_RigVMDispatch_ArrayGetAtIndex = RigVMDispatchNode(Array=RigVMVariableNode() {variable=Controls}.Value, Index=@DISPATCH_RigVMDispatch_ArrayIterator.Index) {template="DISPATCH_RigVMDispatch_ArrayGetAtIndex(in Array,in Index,out Element)"}
   RigVMVariableNode(
    Value=&RigUnit_GetTransform(bInitial=False, Item=@DISPATCH_RigVMDispatch_ArrayGetAtIndex.Element).Transform
    ExecuteContext=@n2.False
   ) {variable=Bone Transform}
   n4 = RigVMDispatchNode(Array=RigVMVariableNode() {variable=Nulls}.Value, Index=@DISPATCH_RigVMDispatch_ArrayIterator.Index) {template="DISPATCH_RigVMDispatch_ArrayGetAtIndex(in Array,in Index,out Element)"}
   &RigUnit_SetTransform(
    Weight=1
    bInitial=False
    Item=@n4.Element
    Value=&RigUnit_GetTransform(bInitial=False, Item=@DISPATCH_RigVMDispatch_ArrayIterator.Element).Transform
    ExecutePin=@n2.True
   )
   > RigVMVariableNode(Value=&RigVMFunction_MathTransformMakeAbsolute(
     Parent=&RigUnit_GetTransform(bInitial=False, Item=@n4.Element).Transform
     Local=&RigUnit_GetTransform(Space=LocalSpace, bInitial=False, Item=@DISPATCH_RigVMDispatch_ArrayGetAtIndex.Element).Transform
    ) {template="Make Absolute::Execute(in Local,in Parent,out Global)"}.Global) {variable=Bone Transform}
  ]}
 RigVMModel Backward Solve Graph {nodes [
   n1 = /Script/ControlRig.RigUnit_InverseExecution() {is=struct RigVMUnitNode}
   RigVMFunction_Sequence = &RigVMFunction_Sequence(ExecuteContext=@n1.ExecutePin)
   DISPATCH_RigVMDispatch_ArrayIterator = RigVMDispatchNode(Array=RigVMVariableNode() {variable=Bones}.Value, ExecuteContext=@RigVMFunction_Sequence.B) {template="DISPATCH_RigVMDispatch_ArrayIterator(in Array,out Element,out Index,out Count,out Ratio)"}
   DISPATCH_RigVMDispatch_ArrayGetAtIndex = RigVMDispatchNode(Index=@DISPATCH_RigVMDispatch_ArrayIterator.Index, Array=RigVMVariableNode() {variable=Controls}.Value) {template="DISPATCH_RigVMDispatch_ArrayGetAtIndex(in Array,in Index,out Element)"}
   > &RigUnit_SetTransform(
    Weight=1
    bInitial=False
    Item=@DISPATCH_RigVMDispatch_ArrayGetAtIndex.Element
    Value=&RigVMFunction_MathTransformMakeAbsolute(
     Parent=&RigUnit_GetTransform(bInitial=False, Item=@DISPATCH_RigVMDispatch_ArrayIterator.Element).Transform
     Local=RigVMVariableNode() {variable=Control Transform Offset}.Value
    ) {template="Make Absolute::Execute(in Local,in Parent,out Global)"}.Global
   )
   Evaluate Space Nulls v01_1 = RigVMFunctionReferenceNode(
    Orient Only=false
    Space Nulls=RigVMVariableNode() {variable=Orient Space Nulls}.Value
    Spaces=RigVMVariableNode() {variable=Orient Spaces}.Value
    ExecuteContext=@RigVMFunction_Sequence.A
   )
   > Evaluate Space Nulls v01_2 = RigVMFunctionReferenceNode(Orient Only=false, Space Nulls=RigVMVariableNode() {variable=Space Nulls}.Value, Spaces=RigVMVariableNode() {variable=Spaces}.Value)
  ]}
 RigVMModel Connector Graph {nodes [/Script/ControlRig.RigUnit_ConnectorExecution() {is=struct RigVMUnitNode}]}
]
hierarchy [
 Parent {type=Connector, loc=<0 0 0>, rot=<0 0 0 1>}
 Root {type=Socket, loc=<0 0 0>, rot=<0 0 0 1>}
 Bones {type=Connector, loc=<0 0 0>, rot=<0 0 0 1>}
 Spaces {type=Connector, loc=<0 0 0>, rot=<0 0 0 1>}
 Orient Spaces {type=Connector, loc=<0 0 0>, rot=<0 0 0 1>}
 Override Parents {type=Connector, loc=<0 0 0>, rot=<0 0 0 1>}
]
variables [
 Bone Transform {type=FTransform, default=(Rotation=(X=0,Y=0,Z=0,W=1),Translation=(X=0,Y=0,Z=0),Scale3D=(X=1,Y=1,Z=1))}
 Bones {type="TArray<FRigElementKey>"}
 CRSL Found {type=bool, default=False}
 Combined Orient Spaces {type="TArray<FRigElementKey>"}
 Combined Spaces {type="TArray<FRigElementKey>"}
 Control Scale {type=double, default=1}
 Control Scale Factor Profile {
  type=FRuntimeFloatCurve
  default=(EditorCurveData=(Keys=((Value=1),(Time=1,Value=1)),DefaultValue=340282346638528859811704183484516925440,PreInfinityExtrap=RCCE_Constant,PostInfinityExtrap=RCCE_Constant),ExternalCurve=None)
 }
 Control Transform Offset {type=FTransform, default=(Rotation=(X=0,Y=0,Z=0,W=1),Translation=(X=0,Y=0,Z=0),Scale3D=(X=1,Y=1,Z=1))}
 Controls {type="TArray<FRigElementKey>"}
 Controls Visibility Initials {type="TArray<bool>"}
 Controls Visibility Initials Sorted {type="TArray<bool>"}
 Display Names {type="TArray<FName>"}
 Display Names Sorted {type="TArray<FName>"}
 Drive Bones {type=bool, default=True}
 FK Control Shape {
  type=FRigUnit_HierarchyAddControl_ShapeSettings
  default=(bVisible=True,Name="Default",Color=(R=1,G=1,B=1,A=1),Transform=(Rotation=(X=0,Y=0,Z=0,W=1),Translation=(X=0,Y=0,Z=0),Scale3D=(X=1,Y=1,Z=1)))
 }
 Fk Parents {type="TArray<FRigElementKey>", default=(())}
 Indices to Remove {type="TArray<int32>"}
 Mirror Axis {type=FVector, default=(X=0,Y=1,Z=1)}
 Mirror Behavior {type=bool, default=True}
 Nulls {type="TArray<FRigElementKey>"}
 Orient Space Nulls {type="TArray<FRigElementKey>"}
 Orient Spaces {type="TArray<FRigElementKey>", default=((Type=Bone,Name="spine_01"),(Type=Bone,Name="spine_02"))}
 Rotation Order {type=EEulerRotationOrder, default=XYZ}
 Space Nulls {type="TArray<FRigElementKey>"}
 Spaces {type="TArray<FRigElementKey>"}
 Standard FK Names {type=bool, default=False}
 Use Active Skeleton {type=bool, default=False}
 Vis Channels {type="TArray<FRigElementKey>"}
]
